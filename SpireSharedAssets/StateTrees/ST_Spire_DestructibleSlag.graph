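# StateTree that drives Spire's DestructibleSlag actor, pulsing an area ability while it senses players.
# PreCombat holds while no player pawns are sensed and moves to Combat on an event once any are.
# Combat's SelectAbility starts the AOEPulse gameplay ability if it can activate, then moves to Wait on completion.
# Wait retries selection every tick, so the ability's activation check paces pulses, and returns to PreCombat when sensed players drop to zero.
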StateTree /SpireSharedAssets/StateTrees/ST_Spire_DestructibleSlag
states [
 Root {
  id=#1
  type=STATE
  children [#2, #3]
  tasks []
  transitions []
 }
 PreCombat {
  id=#2
  type=STATE
  enter_conditions [{
    id=#4
    description="Actor.Sensed Player Pawn Count <= 0"
    properties {
     iD=#5
     instance {_structType=/Script/StateTreeModule.StateTreeCompareIntConditionInstanceData, left=0, right=0}
     node {
      name=None
      _structType=/Script/StateTreeModule.StateTreeCompareIntCondition
      bInvert=false
      deltaIndent=0
      evaluationMode=Evaluated
      operand=And
      operator=LessOrEqual
      bindingsBatch {value=65535}
      executionRuntimeTemplateIndex {value=65535}
      instanceDataHandle {index=65535, source=None, stateHandle {index=65535}}
      instanceTemplateIndex {value=65535}
      outputBindingsBatch {value=65535}
     }
    }
   }]
  tasks [{
    id=#6
    description="Debug Text \"PRECOMBAT\""
    properties {
     iD=#7
     instance {_structType=/Script/StateTreeModule.StateTreeDebugTextTaskInstanceData, bEnabled=false, bindableText="", referenceActor=null}
     node {
      name=None
      _structType=/Script/StateTreeModule.StateTreeDebugTextTask
      bCanEditConsideredForCompletion=false
      bConsideredForCompletion=false
      bHasTaskCompletionDelegateDispatcher=false
      bTaskEnabled=true
      deprecatedVersion=1
      fontScale=1
      text=PRECOMBAT
      transitionHandlingPriority=Normal
      bindingsBatch {value=65535}
      executionRuntimeTemplateIndex {value=65535}
      instanceDataHandle {index=65535, source=None, stateHandle {index=65535}}
      instanceTemplateIndex {value=65535}
      offset=<0 0 0>
      outputBindingsBatch {value=65535}
      textColor=<rgba 1 0 0.9765 1>
     }
    }
   }]
  transitions [{
    id=#8
    trigger=ON_EVENT
    conditions [{
      id=#9
      description="Actor.Sensed Player Pawn Count > 0"
      properties {
       iD=#10
       instance {_structType=&StateTreeCompareIntConditionInstanceData, left=0, right=0}
       node {
        name=None
        _structType=&StateTreeCompareIntCondition
        bInvert=false
        deltaIndent=0
        evaluationMode=Evaluated
        operand=And
        operator=Greater
        bindingsBatch {value=65535}
        executionRuntimeTemplateIndex {value=65535}
        instanceDataHandle {index=65535, source=None, stateHandle {index=65535}}
        instanceTemplateIndex {value=65535}
        outputBindingsBatch {value=65535}
       }
      }
     }]
    state {id=#3, name=Combat}
   }]
 }
 Combat {
  id=#3
  type=STATE
  children [#11, #12]
  tasks [{
    id=#13
    description="Debug Text \"COMBAT\""
    properties {
     iD=#14
     instance {_structType=&StateTreeDebugTextTaskInstanceData, bEnabled=false, bindableText="", referenceActor=null}
     node {
      name=None
      _structType=&StateTreeDebugTextTask
      bCanEditConsideredForCompletion=false
      bConsideredForCompletion=false
      bHasTaskCompletionDelegateDispatcher=false
      bTaskEnabled=true
      deprecatedVersion=1
      fontScale=1
      text=COMBAT
      transitionHandlingPriority=Normal
      bindingsBatch {value=65535}
      executionRuntimeTemplateIndex {value=65535}
      instanceDataHandle {index=65535, source=None, stateHandle {index=65535}}
      instanceTemplateIndex {value=65535}
      offset=<0 0 0>
      outputBindingsBatch {value=65535}
      textColor=<rgba 1 0 0.9765 1>
     }
    }
   }]
  transitions []
 }
 SelectAbility {id=#11, type=STATE, children [#15], tasks [], transitions []}
 Wait {
  id=#12
  type=STATE
  tasks [{
    id=#16
    description=Delay Forever
    properties {
     iD=#17
     instance {_structType=&StateTreeDelayTaskInstanceData, bRunForever=true, duration=1}
     node {name=None, _structType=&StateTreeDelayTask}
    }
   }]
  transitions [
   {
    id=#18
    trigger=ON_TICK
    conditions [{
      id=#19
      description="Actor.Sensed Player Pawn Count == 0"
      properties {
       iD=#20
       instance {_structType=&StateTreeCompareIntConditionInstanceData, left=0, right=0}
       node {
        name=None
        _structType=&StateTreeCompareIntCondition
        bInvert=false
        deltaIndent=0
        evaluationMode=Evaluated
        operand=And
        operator=Equal
        bindingsBatch {value=65535}
        executionRuntimeTemplateIndex {value=65535}
        instanceDataHandle {index=65535, source=None, stateHandle {index=65535}}
        instanceTemplateIndex {value=65535}
        outputBindingsBatch {value=65535}
       }
      }
     }]
    state {id=#2, name=PreCombat}
   }
   {id=#21, trigger=ON_TICK, state {id=#11, name=SelectAbility}}
  ]
 }
 AOEPulse {
  id=#15
  type=STATE
  enter_conditions [{
    id=#22
    description=Can Activate Ability Ability.DestructibleSlag.AOEPulse
    properties {
     iD=#23
     instance {
      _structType=/Script/FortniteAIServer.FortGameplayAbilityCanActivateAbilityInstanceData
      pawn=null
      abilityTag {gameplayTags [{tagName=Ability.DestructibleSlag.AOEPulse}]}
     }
     node {
      name=None
      _structType=/Script/FortniteAIServer.FortGameplayAbilityCanActivateAbilityCondition
      bInvert=false
      deltaIndent=0
      evaluationMode=Evaluated
      operand=And
      bindingsBatch {value=65535}
      executionRuntimeTemplateIndex {value=65535}
      instanceDataHandle {index=65535, source=None, stateHandle {index=65535}}
      instanceTemplateIndex {value=65535}
      outputBindingsBatch {value=65535}
     }
    }
   }]
  tasks [{
    id=#24
    description=Start Ability Ability.DestructibleSlag.AOEPulse
    properties {
     iD=#25
     instance {
      _structType=/Script/FortniteGame.FortStateTreeStartAbilityTaskInstanceData
      actor=null
      abilityTag {gameplayTags [{tagName=Ability.DestructibleSlag.AOEPulse}]}
     }
     node {
      name=None
      _structType=/Script/FortniteGame.FortStateTreeStartAbilityTask
      bCanEditConsideredForCompletion=true
      bConsideredForCompletion=true
      bHasTaskCompletionDelegateDispatcher=false
      bTaskEnabled=true
      transitionHandlingPriority=Normal
      bindingsBatch {value=65535}
      executionRuntimeTemplateIndex {value=65535}
      instanceDataHandle {index=65535, source=None, stateHandle {index=65535}}
      instanceTemplateIndex {value=65535}
      outputBindingsBatch {value=65535}
     }
    }
   }]
  transitions [{id=#26, trigger=ON_STATE_COMPLETED, state {id=#12, name=Wait}}]
 }
]
